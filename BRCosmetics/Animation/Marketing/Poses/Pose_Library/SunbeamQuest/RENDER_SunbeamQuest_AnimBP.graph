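# Marketing render AnimBP for the SunbeamQuest character's pose library, with tail scale variables keyed to a kill count.
# AnimGraph copies the attached parent mesh's pose through the deform Control Rig to the output; a toggleable tail Control Rig sits disconnected.
# Each update, a killCount of 4 (the default) sets tails 1–4 to full scale, then clears activeKill after a delay.
# Branches for the other counts exist but are not wired to the event.
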AnimBlueprint /BRCosmetics/Animation/Marketing/Poses/Pose_Library/SunbeamQuest/RENDER_SunbeamQuest_AnimBP
target_skeleton=/Game/Characters/Player/Female/Medium/Bodies/F_MED_SunbeamQuest/Meshes/F_MED_SunbeamQuest_Skeleton
graphs [
 AnimGraph {nodes [
   Control Rig!(Source=Component To Local(), bAlphaBoolEnabled:bool=Get useTailCtrlRig() -> (useTailCtrlRig:bool)) -> (Pose) {
    alphaInputType=Bool
    alphaBoolBlend {blendInTime=0.25, blendOption=Linear, blendOutTime=1}
    alphaScaleBiasClamp {bInterpResult=true}
    controlRigAssetReference {blueprintRigClass=/Game/Characters/Player/Female/Medium/Bodies/F_MED_SunbeamQuest/Meshes/F_MED_SunbeamQuest_CtrlRig.F_MED_SunbeamQuest_CtrlRig_C}
   }
   n1 = Control Rig!(Source=Copy Pose From Mesh() {bUseAttachedParent=true}, Alpha:float) -> (Pose) {
    alphaBoolBlend {blendOption=Linear}
    controlRigAssetReference {blueprintRigClass=/Game/Characters/Player/Female/Medium/Bodies/F_MED_SunbeamQuest/Meshes/F_MED_SunbeamQuest_Deform_ControlRig.F_MED_SunbeamQuest_Deform_ControlRig_C}
   }
   Output Pose(Result=Component To Local(ComponentPose=Local To Component(LocalPose=@n1))) {layerGroup=None}
  ]}
 EventGraph {nodes [
   Event BlueprintUpdateAnimation()
   > Branch(Condition="Equal (Float)"(A=Get killCount() -> (killCount:double), B=4))
   > Set Tail_1_Scale(Tail_1_Scale:Vector="1,1,1") -> (Output_Get:Vector)
   > Set Tail_2_Scale(Tail_2_Scale:Vector="1,1,1") -> (Output_Get:Vector)
   > Set Tail_3_Scale(Tail_3_Scale:Vector="1,1,1") -> (Output_Get:Vector)
   > n1 = Set Tail_4_Scale(Tail_4_Scale:Vector="1,1,1") -> (Output_Get:Vector)
   Delay(execute=[@n4.then, @n1.then, @n5.then, @n3.then, @n2.then], Duration=0.2)
   > Set activeKill(activeKill:bool=false) -> (Output_Get:bool)
   Branch(Condition="Equal (Float)"(A=Get killCount(), B=3))
   > Set Tail_1_Scale(Tail_1_Scale="1,1,1")
   > Set Tail_2_Scale(Tail_2_Scale="1,1,1")
   > Set Tail_3_Scale(Tail_3_Scale="1,1,1")
   > Set TailPrime_1_Scale(TailPrime_1_Scale:Vector="0, 0, 0") -> (Output_Get:Vector)
   > n2 = Set TailPrime_2_Scale(TailPrime_2_Scale:Vector="0, 0, 0") -> (Output_Get:Vector)
   Branch(Condition="Equal (Float)"(A=Get killCount(), B=1))
   > n3 = Set Tail_1_Scale(Tail_1_Scale="1,1,1")
   Branch(Condition="Equal (Float)"(A=Get killCount(), B=2))
   > Set TailPrime_1_Scale(TailPrime_1_Scale="1,1,1")
   > Set TailPrime_2_Scale(TailPrime_2_Scale="1,1,1")
   > n4 = Set Tail_1_Scale(Tail_1_Scale="0, 0, 0")
   Branch(Condition="Equal (Float)"(A=Get killCount(), B=5))
   > Set Tail_1_Scale(Tail_1_Scale="1,1,1")
   > Set Tail_2_Scale(Tail_2_Scale="1,1,1")
   > Set Tail_3_Scale(Tail_3_Scale="1,1,1")
   > Set Tail_4_Scale(Tail_4_Scale="1,1,1")
   > n5 = Set Tail_5_Scale(Tail_5_Scale:Vector="1,1,1") -> (Output_Get:Vector)
   Reroute Node(InputPin:Wildcard) -> (OutputPin:Wildcard)
   Reroute Node()
  ]}
]
variables [
 activeKill {type=boolean}
 cape {type=boolean}
 killCount {type=number, default=4}
 limitTailRange {type=boolean}
 Local Noise Offset {type=Vector}
 "Local Noise Transform  Location" {type=Vector}
 matInstanceDynamic {type=MaterialInstanceDynamic}
 overlayMID {type=MaterialInstanceDynamic}
 playerPawn {type=FortPlayerPawnAthena}
 R1 Previous Velocity Transform Velocity {type=Vector}
 R1 Velocity Transform Location {type=Vector}
 R2 Velocity Transform Location {type=Vector}
 R2 Velocity Transform Velocity {type=Vector}
 R3 Velocity Transform Location {type=Vector}
 R3 Velocity Transform Velocity {type=Vector}
 R4 Velocity Transform Location {type=Vector}
 R4 Velocity Transform Velocity {type=Vector}
 skelMesh {type=SkeletalMeshComponent}
 SkydiveAlpha {type=number}
 Tail_1_Scale {type=Vector}
 Tail_2_Scale {type=Vector}
 Tail_3_Scale {type=Vector}
 Tail_4_Scale {type=Vector}
 Tail_5_Scale {type=Vector}
 tail_rotate {type=Rotator}
 tail_translate {type=Vector}
 TailPrime_1_Scale {type=Vector}
 TailPrime_2_Scale {type=Vector}
 tailSkydiveAlpha {type=number}
 useTailCtrlRig {type=boolean}
 WPO Pho Wind Intensity {type=number}
]
